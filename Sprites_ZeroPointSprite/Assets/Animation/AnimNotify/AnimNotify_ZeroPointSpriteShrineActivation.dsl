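; Anim notify state for the Zero Point Sprite, named for shrine activation, that swaps materials on the animated skeletal mesh.
; On notify begin it sets slots 0 and 1 to the Zero Point Sprite body and facial materials, and slot 2 to the shared vertex-crunch overlay material.
; Notify end has no logic, so the materials are not restored when the window closes.
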(fn Received_NotifyBegin
  (params (MeshComp :type Skeletal%20Mesh%20Component%20Object%20Reference) (Animation :type Anim%20Sequence%20Base%20Object%20Reference) (TotalDuration :type Float%20%28single-precision%29) (EventReference :type Anim%20Notify%20Event%20Reference%20Structure%20%28by%20ref%29))
  (Rendering|Material|SetMaterial MeshComp 0 "/Sprites_ZeroPointSprite/Assets/Materials/MI_ZeroPoint_Sprite.MI_ZeroPoint_Sprite")
  (Rendering|Material|SetMaterial MeshComp 1 "/Sprites_ZeroPointSprite/Assets/Materials/M_Sprite_ZeroPoint_Facial.M_Sprite_ZeroPoint_Facial")
  (Rendering|Material|SetMaterial MeshComp 2 "/Game/Global/Materials/M_Overlay_VertexCrunch.M_Overlay_VertexCrunch"))

(fn Received_NotifyEnd
  (params (MeshComp :type Skeletal%20Mesh%20Component%20Object%20Reference) (Animation :type Anim%20Sequence%20Base%20Object%20Reference) (EventReference :type Anim%20Notify%20Event%20Reference%20Structure%20%28by%20ref%29)))
