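# Unlit, two-sided translucent Niagara material for the Zero Point Sprite's swirling area effect, tinted and faded by particle color.
# Polar UVs drive panning wave and noise textures for a two-color swirl, sampled with explicit derivatives.
# Procedural masks add a noise-warped glowing edge, radial sine pulses and flickering lightning that scrolls in stepped time.
# Opacity combines radial masks with depth fade, which the mobile shading path skips.
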Material /Sprites_ZeroPointSprite/Assets/FX/Materials/M_SuperSprite_Area_Swirl
graph {
 properties {
  bAutomaticallySetUsageInEditor=false
  bUsedWithNiagaraMeshParticles=true
  bUsedWithNiagaraRibbons=true
  bUsedWithNiagaraSprites=true
  blendMode=BLEND_Translucent
  shadingModel=MSM_Unlit
  twoSided=true
  usedShadingModels=Unlit
 }
 nodes [
  n1 = MaterialFunctionCall!(UV=TextureCoordinate(), Twist Center=Constant(), Twist Edges=Constant()) {materialFunction=/Game/Accessories/FORT_Backpacks/Backpack_F_MED_Celestial_Glow/FX/Materials/MF_PolarUV}
  n2 = MaterialFunctionCall(Float2=NamedRerouteDeclaration(None=@n1.Polar UVs) {name=BaseCoordinates, nodeColor=<rgba 1 0 0.1406 1>}) {materialFunction=&BreakOutFloat2Components}
  LinearInterpolate(Alpha=@n2.G) {constB=0.3}
  MaterialFunctionCall() {materialFunction=&VectorToRadialValue}
  n3 = ParticleColor()
  n4 = LinearInterpolate(Alpha=@n2.G) {constA=0.1}
  n5 = Multiply(A=NamedRerouteUsage(), B=Constant2Vector() {g=0.3, r=1})
  n6 = NamedRerouteDeclaration(None=TextureSample!(
    UVs=Panner(Coordinate=@n5) {speedX=-0.05, speedY=-0.15}
    Tex=TextureObjectParameter() {group=Waves, parameterName=TextureWave, texture=/Game/Effects/Fort_Effects/Textures/Tile/T_CircleNoise_01_M}
    "DDX(UVs)"=NamedRerouteUsage()
    "DDY(UVs)"=NamedRerouteUsage()
    Apply View MipBias
   ) {mipValueMode=TMVM_Derivative, texture=/Game/Effects/Fort_Effects/Textures/Tile/T_SpookyFaces_Chroma_Tile}.R) {name=Waves, nodeColor=<rgba 0 0.3672 1 1>}
  n7 = NamedRerouteUsage()
  n8 = MaterialFunctionCall(UV=TextureCoordinate(), Twist Center=Constant() {r=0.15}, Twist Edges=Constant()) {materialFunction=&MF_PolarUV}
  n9 = Power(Base=Multiply(A=@n8.Polar UVs, B=Constant2Vector() {g=0.6, r=2}), Exp=Constant2Vector() {g=2, r=1})
  n10 = Power(Base=NamedRerouteUsage()) {constExponent=0.5}
  n11 = Constant2Vector() {g=0.4}
  n12 = TextureObjectParameter() {group=Waves, parameterName=TextureNoise, samplerType=SAMPLERTYPE_LinearColor, texture=/Game/Global/Textures/Default/Noise/T_Curie_NoisePack}
  n13 = TextureSample(UVs=Add(A=Panner(Coordinate=@n9) {speedX=-0.09, speedY=-0.15}, B=Multiply(A=@n10, B=@n11)), Tex=@n12, "DDX(UVs)"=@n8.DDX, "DDY(UVs)"=@n8.DDX) {mipValueMode=TMVM_Derivative, texture=&T_SpookyFaces_Chroma_Tile}
  n14 = Add(A=@n13.G, B=Multiply(A=@n13.B) {constB=1.5})
  n15 = Multiply(A=@n14) {constB=4}
  n16 = Max(A=@n6, B=Multiply(A=LinearInterpolate(Alpha=@n6) {constA=0.2}, B=Multiply(A=@n7, B=@n15)) {constA=1})
  n17 = ScalarParameter() {defaultValue=1.5, group=Waves, parameterName=WavesColorPower}
  n18 = LinearInterpolate(
   A=VectorParameter() {group=Waves, parameterName=WavesColorDark, defaultValue=<rgba 0.1 0.0333 0.02 0>}.RGB
   B=VectorParameter() {group=Waves, parameterName=WavesColorLight, defaultValue=<rgba 1 0.71 0.3 0>}.RGB
   Alpha=Multiply(A=@n4, B=Power(Base=@n16, Exp=@n17) {constExponent=1.5})
  )
  n19 = NamedRerouteUsage()
  n20 = Add(A=@n18, B=@n19)
  n21 = Add(A=@n20, B=Multiply(A=NamedRerouteUsage(), B=ScalarParameter() {defaultValue=0.6, parameterName=Edge Color Intensity}) {constB=0.6})
  n22 = VectorParameter() {parameterName=Lightning Color, defaultValue=<rgba 7 5.6 2.8 1>}
  n23 = NamedRerouteUsage()
  MaterialFunctionCall(Emissive=Multiply(A=@n3.RGB, B=LinearInterpolate(A=@n21, B=@n22.RGB, Alpha=@n23))) {materialFunction=&MF_FX_EmissiveAdjustments}
  n24 = MaterialFunctionCall(Float2=NamedRerouteUsage()) {materialFunction=&BreakOutFloat2Components}
  n25 = OneMinus(None=@n24.G)
  n26 = NamedRerouteUsage()
  Max(A=@n25, B=@n26)
  Max(A=@n18, B=@n19)
  n27 = Saturate(None=Multiply(A=Power(Base=@n24.G) {constExponent=5}))
  n28 = Saturate(None=Multiply(A=Power(Base=@n24.G) {constExponent=10}) {constB=10})
  Max(A=@n27, B=@n28)
  Multiply(A=@n20, B=@n4)
  n29 = Add(A=@n25, B=@n26)
  n30 = Multiply(A=Saturate(None=Add(A=@n27, B=@n28)), B=Max(A=NamedRerouteUsage(), B=Saturate(None=@n29)))
  n31 = DepthFade(FadeDistance=ScalarParameter() {defaultValue=60, parameterName=DepthFadeDist})
  Multiply(
   A=Multiply(A=@n3.A, B=Saturate(None=@n30))
   B=ShadingPathSwitch(Default=@n31, Deferred=@n31, Mobile=Constant() {r=1}) {desc=Migrated from FeatureLevelSwitch}
  )
  n32 = Power(Base=NamedRerouteUsage()) {constExponent=3}
  n33 = SphereMask(A=TextureCoordinate(), B=Constant() {r=0.5}) {attenuationRadius=0.46, hardnessPercent=0}
  NamedRerouteDeclaration(None=SmoothStep(Value=Max(A=@n32, B=@n33)) {constMax=0.1}) {name=AoEMask, nodeColor=<rgba 0.9922 1 0 1>}
  NamedRerouteDeclaration(None=@n1.DDX) {name=PolarUV_DDX, nodeColor=<rgba 1 0.0789 0.0598 1>}
  NamedRerouteDeclaration(None=@n1.DDY) {name=PolarUV_DDY, nodeColor=<rgba 0.2434 1 0.0703 1>}
  NamedRerouteDeclaration(None=@n14) {name=Noise, nodeColor=<rgba 0 1 0.3203 1>}
  n34 = NamedRerouteUsage()
  n35 = Time()
  n36 = Multiply(A=@n35) {constB=3}
  n37 = Multiply(A=Add(A=@n35) {constB=5}) {constB=0.6}
  n38 = Max(A=ConstantBiasScale(None=Sine(None=@n36)), B=ConstantBiasScale(None=Cosine(None=@n37)))
  n39 = LinearInterpolate(Alpha=Power(Base=NamedRerouteUsage()) {constExponent=3}) {constA=0.5}
  n40 = Multiply(
   A=ScalarParameter() {defaultValue=0.1, parameterName=CircleWarpIntensity}
   B=LinearInterpolate(Alpha=Multiply(A=@n38, B=@n39)) {constA=1, constB=2}
  )
  n41 = Add(A=@n24.G, B=Multiply(A=@n24.G, B=Multiply(A=@n34, B=@n40) {constB=0.2})) {constB=0}
  n42 = SphereMask(A=Frac(None=OneMinus(None=@n41)), B=Constant() {r=0.1}) {attenuationRadius=0.08, hardnessPercent=0}
  n43 = Power(Base=@n42) {constExponent=18}
  NamedRerouteDeclaration(None=LinearInterpolate(A=Power(Base=@n42) {constExponent=1.5}, B=Multiply(A=@n43)) {constAlpha=0.65}) {name=Edge, nodeColor=<rgba 0 0.7422 1 1>}
  n44 = VectorParameter() {group=Lightning, parameterName=LightningUVs, defaultValue=<rgba 1 0.4 -0.3 -0.95>}
  n45 = Time()
  n46 = ScalarParameter() {defaultValue=2, group=Lightning, parameterName=LightningCeilSteps}
  n47 = Divide(A=Ceil(None=Multiply(A=@n45, B=@n46)), B=@n46)
  n48 = Add(
   A=Multiply(A=NamedRerouteUsage(), B=AppendVector(A=@n44.R, B=@n44.R))
   B=AppendVector(A=Multiply(A=@n47, B=@n44.B) {constB=-0.3}, B=Multiply(A=@n47, B=@n44.A) {constB=-0.95})
  )
  n49 = TextureObjectParameter() {
   group=Lightning
   parameterName=TextureLightning
   samplerType=SAMPLERTYPE_Grayscale
   texture=/Sprites_ZeroPointSprite/Assets/FX/Textures/T_SuperSprite_LightningTile
  }
  n50 = NamedRerouteUsage()
  n51 = NamedRerouteUsage()
  n52 = Time()
  n53 = Sine(None=Multiply(A=@n52) {constB=4})
  n54 = ConstantBiasScale(None=Sine(None=Multiply(A=@n52) {constB=0.7}))
  n55 = Multiply(
   A=Power(Base=TextureSample(UVs=@n48, Tex=@n49, "DDX(UVs)"=@n50, "DDY(UVs)"=@n51) {mipValueMode=TMVM_Derivative, texture=&T_SpookyFaces_Chroma_Tile}.R) {constExponent=1}
   B=Multiply(A=ConstantBiasScale(None=@n53), B=Power(Base=@n54) {constExponent=7})
  )
  n56 = Multiply(A=Multiply(A=Power(Base=@n55) {constExponent=1}) {constB=3}) {constB=0.7}
  n57 = Power(Base=Multiply(A=@n55) {constB=2}) {constExponent=4}
  NamedRerouteDeclaration(None=Multiply(A=NamedRerouteUsage(), B=Add(A=@n56, B=@n57))) {name=Lightning, nodeColor=<rgba 0 1 0.5078 1>}
  n58 = Power(Base=MaterialFunctionCall(Float2=NamedRerouteUsage()) {materialFunction=&BreakOutFloat2Components}.G) {constExponent=3}
  n59 = Multiply(A=Time()) {constB=-1.6}
  n60 = Sine(None=Multiply(A=Add(A=@n58, B=@n59)) {constB=0.5})
  n61 = Power(Base=@n60) {constExponent=5}
  n62 = Power(Base=@n60) {constExponent=100}
  NamedRerouteDeclaration(None=Multiply(A=LinearInterpolate(A=@n61, B=@n62), B=NamedRerouteUsage())) {name=Pulse, nodeColor=<rgba 0.1941 0.3537 1 1>}
  NamedRerouteUsage()
  NamedRerouteUsage()
 ]
}
