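# Unlit, two-sided material for Niagara mesh particles that draws the glowing orange SuperSprite charge symbol.
# The grayscale symbol texture, shaped by several power curves, blends three color parameters, with brightness scaled by a polar-coordinate gradient.
# Panned, rotated UVs with a per-particle random offset feed high-power sine bands, creating moving glints masked by gradient and symbol.
# These glints push the color toward HDR Color 3 before MF_FX_EmissiveAdjustments.
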Material /Sprites_ZeroPointSprite/Assets/FX/Materials/M_SuperSprite_ChargeSymbol
graph {
 properties {bAutomaticallySetUsageInEditor=false, bUsedWithNiagaraMeshParticles=true, shadingModel=MSM_Unlit, twoSided=true, usedShadingModels=Unlit}
 nodes [
  n1 = TextureCoordinate()
  Clamp(None=MaterialFunctionCall(UVs=@n1, Rotation Center=Constant() {r=0.5}, "Rotation Angle (0-1)"=Constant() {r=0.125}) {materialFunction=&CustomRotator}.Rotated Values)
  Constant()
  Constant() {r=1}
  n2 = MaterialFunctionCall!(UV Coord=TextureCoordinate(), LenghtScale, RadialScale) {materialFunction=/Game/Global/MaterialFunctions/TechArt/MF_PolarCoordinates}
  n3 = Multiply(A=MaterialFunctionCall(Float2=Panner(Coordinate=@n2.Result) {speedY=-1.6}) {materialFunction=&BreakOutFloat2Components}.R) {constB=2}
  n4 = Saturate(None=LinearInterpolate(Alpha=Saturate(None=@n3)) {constA=-3.5, constB=3})
  n5 = LinearInterpolate(Alpha=@n4) {constA=2.5}
  n6 = VectorParameter() {parameterName=Color 1, defaultValue=<rgba 1 0.1917 0.03 1>}
  n7 = Multiply(A=@n6.RGB) {constB=0.85}
  n8 = TextureSampleParameter2D(UVs=@n1) {
   parameterName=SymbolTexture2
   samplerType=SAMPLERTYPE_LinearGrayscale
   texture=/Sprites_ZeroPointSprite/Assets/FX/Textures/T_SuperSprite_SymbolCharge
  }
  n9 = Power(Base=@n8.R) {constExponent=0.5}
  n10 = VectorParameter() {parameterName=Color 3, defaultValue=<rgba 170 56.6667 0 1>}
  n11 = LinearInterpolate(
   A=LinearInterpolate(
    A=LinearInterpolate(A=@n7, B=@n6.RGB, Alpha=@n9)
    B=VectorParameter() {parameterName=Color 2, defaultValue=<rgba 0.7 0.3325 0.07 1>}.RGB
    Alpha=Power(Base=@n8.R) {constExponent=0.8}
   )
   B=@n10.RGB
   Alpha=Multiply(A=Power(Base=@n8.R) {constExponent=11}) {constB=0.16}
  )
  n12 = TextureCoordinate()
  n13 = ParticleRandom()
  n14 = Panner(Coordinate=Multiply(A=Add(A=@n12, B=@n13)) {constB=1.5}) {speedY=3}
  n15 = Constant() {r=0.5}
  n16 = Constant() {r=0.2}
  n17 = Multiply(A=MaterialFunctionCall(Float2=MaterialFunctionCall(UVs=@n14, Rotation Center=@n15, "Rotation Angle (0-1)"=@n16) {materialFunction=&CustomRotator}.Rotated Values) {materialFunction=&BreakOutFloat2Components}.R) {constB=0.5}
  n18 = Sine(None=@n17)
  n19 = Power(Base=ConstantBiasScale(None=@n18)) {constExponent=100}
  n20 = Add(A=@n17) {constB=0.12}
  n21 = Power(Base=ConstantBiasScale(None=Sine(None=@n20))) {constExponent=200}
  n22 = Saturate(None=Multiply(A=Max(A=@n19, B=@n21)))
  n23 = Multiply(A=LinearInterpolate(Alpha=Multiply(A=@n22)), B=LinearInterpolate(Alpha=Power(Base=@n8.R) {constExponent=4}) {constA=0.02, constAlpha=0.9})
  MaterialFunctionCall(Emissive=LinearInterpolate(A=Multiply(A=@n5, B=@n11), B=@n10.RGB, Alpha=Multiply(A=@n4, B=@n23))) {materialFunction=&MF_FX_EmissiveAdjustments}
  MaterialFunctionCall() {materialFunction=&FixRotateAboutAxisNormals}
  MaterialFunctionCall(Float2=@n2.Result) {materialFunction=&BreakOutFloat2Components}
  PerInstanceRandom()
  Power(Base=@n12) {constExponent=5}
  n24 = Multiply(A=Time()) {constB=0.2}
  Power(Base=ConstantBiasScale(None=Sine(None=@n24))) {constExponent=5}
  n25 = Time()
  n26 = ParticleRandom()
  RotateAboutAxis(
   NormalizedRotationAxis=Normalize(VectorInput=Constant3Vector() {constant=<rgba 0 0 1 0>})
   RotationAngle=Multiply(A=Add(A=@n25, B=@n26)) {constA=1, constB=-0.3}
   PivotPoint=TransformPosition(None=Constant3Vector()) {transformType=TRANSFORMPOSSOURCE_World}
   Position=Multiply(A=WorldPosition().XYZ)
  )
  Saturate(None=@n18)
  ScreenPosition()
  VertexNormalWS()
 ]
}
