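# Unlit, two-sided translucent trail material for Niagara ribbons and mesh particles, named for the Zero Point sprite's dash idle effect.
# World position offset combines SplineThicken, driven by dynamic Width, with a time-scrolled sine wave shifted by dynamic Offset.
# A panned trail texture shapes particle-tinted emissive color and merges with edge-masked caustic noise for opacity.
# Opacity also uses a sine falloff along U and depth fade, which forward and mobile paths skip.
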Material /Sprites_ZeroPointSprite/Assets/FX/Materials/M_SuperSprite_DashIdle_Trail
graph {
 properties {
  bAutomaticallySetUsageInEditor=false
  bUsedWithNiagaraMeshParticles=true
  bUsedWithNiagaraRibbons=true
  blendMode=BLEND_Translucent
  shadingModel=MSM_Unlit
  twoSided=true
  usedShadingModels=Unlit
 }
 nodes [
  n1 = DynamicParameter() {paramNames [Width,Offset,Param3,Param4]}
  n2 = Transform(None=VectorParameter() {parameterName=WPO_Direction, defaultValue=<rgba 0 1 0 1>}.RGB) {transformSourceType=TRANSFORMSOURCE_Local, transformType=TRANSFORM_View}
  n3 = Add(A=TextureCoordinate() {uTiling=5}, B=DynamicParameter() {paramNames [Width,Offset,Param3,Param4]}.Offset)
  n4 = Time()
  n5 = Sine(None=Add(A=ComponentMask(None=@n3) {r=true}, B=Multiply(A=@n4) {constB=-6})) {period=3}
  Add(
   A=MaterialFunctionCall(WidthBase=@n1.Width, WidthTip=@n1.Width, Expand U or V UV Channel=StaticBool()) {materialFunction=&SplineThicken}.WorldPositionOffset
   B=Multiply(A=Multiply(A=@n2, B=@n5), B=ScalarParameter() {defaultValue=10, parameterName=WPOIntensity})
  )
  n6 = TextureSample(UVs=Panner(Coordinate=TextureCoordinate()) {speedX=0.1}) {samplerType=SAMPLERTYPE_Normal, texture=/Game/Characters/Player/Female/Medium/Bodies/F_MED_Bistro_Astronaut/FX/T_GaussianNoise_Normal}
  n7 = Add(A=TextureCoordinate(), B=Multiply(A=@n6.G, B=ScalarParameter() {defaultValue=0.2, parameterName=Disturb}) {constB=0.2})
  n8 = ScalarParameter() {defaultValue=0.01, parameterName=ColorOffset}
  n9 = Constant2Vector() {g=0.3, r=0.3}
  n10 = TextureSampleParameter2D(UVs=Panner(Coordinate=Add(A=@n7, B=@n8), Speed=@n9)) {parameterName=Color_Tex, texture=/Game/Global/Textures/Default/Caustics/T_Nosie_Caustics_Grayscale}
  n11 = TextureSampleParameter2D(UVs=Panner(Coordinate=@n7, Speed=@n9)) {parameterName=Color_Tex, texture=&T_Nosie_Caustics_Grayscale}
  n12 = Power(Base=Desaturation(
    None=MaterialFunctionCall(X=@n10.R, Y=@n11.R) {materialFunction=&MakeFloat3}.Result
    Fraction=ScalarParameter() {defaultValue=-2, parameterName=Des}
   )) {constExponent=0.3}
  n13 = TextureCoordinate()
  n14 = Multiply(A=@n13, B=ComponentMask(None=VectorParameter() {parameterName=Default_UVScale, defaultValue=<rgba 1 1 0 0>}.RGB) {g=true, r=true})
  n15 = VectorParameter() {parameterName=Default_Panner, defaultValue=<rgba -2 0 0 0>}
  n16 = TextureSampleParameter2D(UVs=Panner(Coordinate=Add(A=@n14) {constB=0}, Speed=ComponentMask(None=@n15.RGB) {g=true, r=true}) {speedX=-2}) {parameterName=Default_Tex, texture=/Game/Weapons/FORT_Melee/Pickaxe_Sunlit/FX/T_Pickaxe_Sunlit_Trail_01}
  n17 = ScalarParameter() {defaultValue=10, parameterName=Power}
  Add(A=Multiply(B=Multiply(A=@n12, B=@n16.R)) {constA=1}, B=Multiply(A=Power(Base=@n16.R, Exp=@n17) {constExponent=10}))
  n18 = MaterialFunctionCall(Float2=@n13) {materialFunction=&BreakOutFloat2Components}
  n19 = Sine(None=Multiply(A=@n18.G) {constB=0.5})
  ConstantBiasScale(None=@n19)
  Constant() {r=30}
  n20 = TextureCoordinate()
  n21 = Constant2Vector() {g=1, r=0.35}
  n22 = TextureSampleParameter2D(UVs=Panner(Coordinate=Multiply(A=@n20, B=@n21)) {speedX=-0.55}) {parameterName=Color_Tex, samplerType=SAMPLERTYPE_LinearGrayscale, texture=&T_Nosie_Caustics_Grayscale}
  n23 = OneMinus(None=Multiply(A=@n18.R))
  n24 = Multiply(A=Power(Base=@n19) {constExponent=10}, B=Saturate(None=Power(Base=@n23) {constExponent=1}))
  n25 = Multiply(A=@n22.R, B=Multiply(A=@n24) {constB=2})
  LinearInterpolate(A=@n25, Alpha=@n24)
  n26 = Constant3Vector() {constant=<rgba 0.75 0.5 0.15 1>}
  n27 = Multiply(A=Power(Base=@n16.R) {constExponent=1})
  n28 = Power(Base=@n27)
  MaterialFunctionCall(Emissive=Multiply(A=ParticleColor().RGB, B=LinearInterpolate(A=@n26, Alpha=@n28) {constB=1.4})) {materialFunction=&MF_FX_EmissiveAdjustments}
  n29 = Max(A=@n25, B=@n27)
  n30 = Sine(None=ComponentMask(None=TextureCoordinate()) {r=true}) {period=2}
  n31 = DepthFade(FadeDistance=ScalarParameter() {defaultValue=15, parameterName=FadeDistance}) {fadeDistanceDefault=10}
  n32 = Constant() {r=1}
  Multiply(A=Multiply(A=ParticleColor().A, B=Multiply(A=@n29, B=@n30)), B=ShadingPathSwitch(Default=@n31, Deferred=@n31, Forward=@n32, Mobile=@n32))
  Panner(Coordinate=Subtract(A=@n7, B=@n8), Speed=@n9)
  Saturate(None=Multiply(A=@n30) {constB=2})
 ]
}
