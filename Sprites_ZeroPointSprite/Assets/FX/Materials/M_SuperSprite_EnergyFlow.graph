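# Translucent, unlit, two-sided effect material for Niagara mesh particles that draws a scrolling energy flow.
# Two panning fire-noise samples are blended, brightened toward white by a V-gradient glow, then masked by adjustable top and bottom fades.
# Particle color tints that mask for emissive, while particle alpha, inverted Fresnel and depth fade (skipped on mobile) shape opacity.
# The noise also scales vertex normals for world-position-offset displacement.
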Material /Sprites_ZeroPointSprite/Assets/FX/Materials/M_SuperSprite_EnergyFlow
graph {
 properties {
  bAutomaticallySetUsageInEditor=false
  bUsedWithNiagaraMeshParticles=true
  blendMode=BLEND_Translucent
  shadingModel=MSM_Unlit
  twoSided=true
  usedShadingModels=Unlit
 }
 nodes [
  n1 = TextureCoordinate()
  n2 = Multiply(A=@n1, B=Constant2Vector() {g=0.1, r=1})
  n3 = Add(A=Multiply(A=@n1, B=Constant2Vector() {g=0.3, r=1})) {constB=0.2}
  n4 = LinearInterpolate(
   A=TextureSampleParameter2D(UVs=Panner(Coordinate=@n2) {speedY=0.1}) {parameterName=TEx1, texture=/Game/Global/Textures/Default/Noise/T_FireNoise_01}.R
   B=TextureSampleParameter2D(UVs=Panner(Coordinate=@n3) {speedX=0.03, speedY=0.07}) {parameterName=TEx1, texture=&T_FireNoise_01}.G
  )
  n5 = Multiply(A=@n4) {constB=0.7}
  n6 = Multiply(A=MaterialFunctionCall(Float2=@n1) {materialFunction=&BreakOutFloat2Components}.G)
  n7 = LinearInterpolate(A=@n4, Alpha=Power(Base=Add(A=@n5, B=@n6), Exp=ScalarParameter() {defaultValue=6, parameterName=GlowPower}) {constExponent=6}) {constA=1}
  n8 = ScalarParameter() {defaultValue=1, parameterName=TopFade_Power}
  n9 = Saturate(None=Multiply(A=Power(Base=@n6, Exp=@n8) {constExponent=1}, B=ScalarParameter() {defaultValue=1, parameterName=TopFade_Harshness}))
  n10 = OneMinus(None=@n6)
  n11 = ScalarParameter() {defaultValue=1, parameterName=BottomFade_Power}
  n12 = Saturate(None=Multiply(A=Power(Base=@n10, Exp=@n11) {constExponent=1.1}, B=ScalarParameter() {defaultValue=5, parameterName=BottomFade_Harshness}) {constB=5})
  n13 = Multiply(A=@n7, B=Multiply(A=Multiply(A=@n9, B=@n12), B=ScalarParameter() {defaultValue=5, parameterName=Fade_Intensity}) {constB=5})
  Add(A=Power(Base=@n4) {constExponent=10}, B=@n13)
  LinearInterpolate(
   A=LinearInterpolate(
    A=Constant3Vector() {constant=<rgba 0.1979 0.0018 0 1>}
    B=Constant3Vector() {constant=<rgba 1 1 1 0>}
    Alpha=SmoothStep(Value=@n7) {constMin=0.9}
   )
   B=@n13
  ) {constAlpha=0.2}
  n14 = ParticleColor()
  MaterialFunctionCall(Emissive=Multiply(A=@n13, B=@n14.RGB)) {materialFunction=&MF_FX_EmissiveAdjustments}
  Multiply(A=Multiply(A=@n4, B=VertexNormalWS()), B=ScalarParameter() {defaultValue=50, parameterName=WPO_Intensity}) {constB=100}
  n15 = Fresnel() {baseReflectFraction=0, exponent=1}
  n16 = DepthFade(FadeDistance=ScalarParameter() {defaultValue=80, parameterName=DepthFadeDist})
  Multiply(
   A=Multiply(A=OneMinus(None=@n15), B=Multiply(A=@n13, B=@n14.A))
   B=ShadingPathSwitch(Default=@n16, Deferred=@n16, Mobile=Constant() {r=1}) {desc=Migrated from FeatureLevelSwitch}
  )
 ]
}
