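# Unlit skeletal-mesh material for the facial parts of the ZeroPoint Sprite character, built around an eye-white texture.
# The texture is brightened and divided by eye adaptation to offset auto-exposure.
# A pre-skinning local-height gradient, interpolated per vertex for pixel shading, is blended by vertex-color red.
# A parameterized, scale-aware offset along the camera vector is built for pushing vertices toward the camera.
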Material /Sprites_ZeroPointSprite/Assets/Materials/M_Sprite_ZeroPoint_Facial
graph {
 properties {
  bAutomaticallySetUsageInEditor=false
  bUsedWithMeshDeformer=true
  bUsedWithSkeletalMesh=true
  shadingModel=MSM_Unlit
  usedShadingModels=Unlit
 }
 nodes [
  Constant() {r=0.1}
  Divide(
   A=Multiply(A=TextureSampleParameter2D() {parameterName=EyeWhite, texture=/Sprites_ZeroPointSprite/Assets/Textures/T_Sprite_EYE}.RGB) {constB=5}
   B=EyeAdaptation().EyeAdaptation
  )
  n1 = LocalPosition() {localOrigin=InstancePreSkinning}
  n2 = ShaderStageSwitch(PixelShader=VertexInterpolator(VS=@n1.Z).PS, VertexShader=@n1.Z)
  n3 = VertexColor()
  LinearInterpolate(A=SmoothStep(Value=Divide(A=@n2) {constB=50}) {constMax=0.6, constMin=0.42}, Alpha=@n3.R)
  Multiply(
   A=Multiply(A=MaterialFunctionCall() {materialFunction=&CameraDirectionVector}.Result) {constB=-1}
   B=Multiply(
    A=ScalarParameter() {parameterName=WPO_PushToCameraAmount}
    B=MaterialFunctionCall() {materialFunction=&ObjectScale}.Scale Z
   )
  )
  Multiply(A=@n3.R)
 ]
}
