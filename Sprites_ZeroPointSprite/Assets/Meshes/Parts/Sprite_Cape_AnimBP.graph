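# Animation Blueprint for the Zero Point Sprite's cape, combining a played animation with simulated secondary motion.
# The generic sidekick cape sequence plays, gets an additive downward shift on the offset bone, then passes through DefaultSlot for montages.
# Four Anim Dynamics nodes in series then simulate single cape bones (08, 13, 18, 23) in root-relative space with angular springs, limits and overridden gravity.
# The event graph holds no logic.
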AnimBlueprint /Sprites_ZeroPointSprite/Assets/Meshes/Parts/Sprite_Cape_AnimBP
target_skeleton=&/Sprite_Cape_Creature_Scale_Skeleton
graphs [
 AnimGraph {nodes [
   n1 = "Transform (Modify) Bone - Bone: offset"(
    ComponentPose=Local To Component(LocalPose=Sequence Player 'Sprite_Sidekick_Generic_Cape'() {sequence=/Sprites_ZeroPointSprite/Assets/Animation/Sprite_Sidekick_Generic_Cape})
    Alpha=1
    Translation="0,0,-7"
    Rotation="0,0,0"
    Scale="1,1,1"
   ) {
    translationMode=BMM_Additive
    translationSpace=BCS_ParentBoneSpace
    alphaBoolBlend {blendOption=Linear}
    boneToModify {boneName=offset}
    translation {z=-7}
   }
   n2 = Component To Local(ComponentPose=@n1)
   n3 = Anim Dynamics - Bone: cape_bind_08(ComponentPose=Local To Component(LocalPose=Slot 'DefaultSlot'(Source=@n2)), Alpha=1, ExternalForce="0,0,0") {
    angularDampingOverride=1
    angularSpringConstant=5000
    bAngularSpring=true
    bOverrideAngularDamping=true
    bUseGravityOverride=true
    gravityScale=0.35
    simulationSpace=RootRelative
    alphaBoolBlend {blendOption=Linear}
    boundBone {boneName=cape_bind_08}
    componentLinearAccScale=<1 1 1>
    componentLinearVelScale=<1 1 1>
    gravityOverride {z=-1500}
    physicsBodyDefinitions [{
      boundBone {boneName=cape_bind_08}
      boxExtents=<15 15 15>
      constraintSetup {angularTargetAxis=AxisY, angularLimitsMax=<30 30 30>, angularLimitsMin=<-30 -30 -30>, angularTarget=<0 1 0>}
      localJointOffset=<0 -1 0>
     }]
   }
   n4 = Anim Dynamics - Bone: cape_bind_13(ComponentPose=@n3, Alpha=1, ExternalForce="0,0,0") {
    angularDampingOverride=1
    angularSpringConstant=5000
    bAngularSpring=true
    bOverrideAngularDamping=true
    bUseGravityOverride=true
    gravityScale=0.35
    simulationSpace=RootRelative
    alphaBoolBlend {blendOption=Linear}
    boundBone {boneName=cape_bind_13}
    componentLinearAccScale=<1 1 1>
    componentLinearVelScale=<1 1 1>
    gravityOverride {z=-1500}
    physicsBodyDefinitions [{
      boundBone {boneName=cape_bind_13}
      boxExtents=<15 15 15>
      constraintSetup {angularTargetAxis=AxisY, angularLimitsMax=<30 30 30>, angularLimitsMin=<-30 -30 -30>, angularTarget=<0 1 0>}
      localJointOffset=<0 -1 0>
     }]
   }
   n5 = Anim Dynamics - Bone: cape_bind_18(ComponentPose=@n4, Alpha=1, ExternalForce="0,0,0") {
    angularDampingOverride=1
    angularSpringConstant=5000
    bAngularSpring=true
    bOverrideAngularDamping=true
    bUseGravityOverride=true
    gravityScale=0.35
    simulationSpace=RootRelative
    alphaBoolBlend {blendOption=Linear}
    boundBone {boneName=cape_bind_18}
    componentLinearAccScale=<1 1 1>
    componentLinearVelScale=<1 1 1>
    gravityOverride {z=-1500}
    physicsBodyDefinitions [{
      boundBone {boneName=cape_bind_18}
      boxExtents=<15 15 15>
      constraintSetup {angularTargetAxis=AxisY, angularLimitsMax=<30 30 30>, angularLimitsMin=<-30 -30 -30>, angularTarget=<0 1 0>}
      localJointOffset=<0 -1 0>
     }]
   }
   n6 = Anim Dynamics - Bone: cape_bind_23(ComponentPose=@n5, Alpha=1, ExternalForce="0,0,0") {
    angularDampingOverride=1
    angularSpringConstant=5000
    bAngularSpring=true
    bOverrideAngularDamping=true
    bUseGravityOverride=true
    gravityScale=0.35
    simulationSpace=RootRelative
    alphaBoolBlend {blendOption=Linear}
    boundBone {boneName=cape_bind_23}
    componentLinearAccScale=<1 1 1>
    componentLinearVelScale=<1 1 1>
    gravityOverride {z=-1500}
    physicsBodyDefinitions [{
      boundBone {boneName=cape_bind_23}
      boxExtents=<15 15 15>
      constraintSetup {angularTargetAxis=AxisY, angularLimitsMax=<30 30 30>, angularLimitsMin=<-30 -30 -30>, angularTarget=<0 1 0>}
      localJointOffset=<0 -1 0>
     }]
   }
   Output Pose(Result=Component To Local(ComponentPose=@n6))
  ]}
 EventGraph {nodes [
   TryGetPawnOwner()
   Event BlueprintUpdateAnimation()
  ]}
]
variables []
